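# Control Rig for the Cactus Rocker backpack cosmetic, adding procedural secondary motion to its creature-like jaw.
# Begin Execution drives a single Chain Harmonics unit on the bone chain rooted at dyn_jaw_1.
# That unit blends pendulum swing with a noisy wave, with reach also enabled and debug drawing left on.
# The imported hierarchy also holds spine, tail-like tile and fin bones, which this graph leaves unanimated.
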ControlRigBlueprint /BRCosmetics/Accessories/FORT_Backpacks/Backpack_M_MED_Cactus_Rocker/Meshes/M_MED_Cactus_Rocker_Pack_CtrlRig
graphs [
 RigVMFunctionLibrary {links [], nodes []}
 RigVMModel {nodes [
   RigUnit_BeginExecution = &RigUnit_BeginExecution()
   /Script/ControlRig.RigUnit_ChainHarmonicsPerItem(
    ChainRoot=(Type=Bone,Name="dyn_jaw_1")
    DrawWorldOffset=(Rotation=(X=0,Y=0,Z=0,W=1),Translation=(X=0,Y=0,Z=0),Scale3D=(X=1,Y=1,Z=1))
    Pendulum=(bEnabled=True,PendulumStiffness=2.5,PendulumGravity=(X=0,Y=0,Z=30),PendulumBlend=1,PendulumDrag=3,PendulumMaximum=0.1,UnwindAxis=(X=0,Y=1,Z=0),UnwindMinimum=0.2,UnwindMaximum=0.05)
    Reach=(bEnabled=True,ReachTarget=(X=0,Y=0,Z=0),ReachAxis=(X=0,Y=0,Z=0))
    Speed=(X=0,Y=0,Z=1)
    Wave=(bEnabled=True,WaveFrequency=(X=0,Y=0,Z=0),WaveAmplitude=(X=0,Y=0,Z=0.3),WaveOffset=(X=0,Y=0,Z=0),WaveNoise=(X=0,Y=0,Z=3),WaveMaximum=1,WaveEase=CubicEaseInOut)
    WaveCurve=(EditorCurveData=(Keys=((),(Time=1,Value=1)),DefaultValue=340282346638528859811704183484516925440,PreInfinityExtrap=RCCE_Constant,PostInfinityExtrap=RCCE_Constant))
    bDrawDebug=True
    ExecutePin=@RigUnit_BeginExecution.ExecutePin
   ) {is=struct RigVMUnitNode}
  ]}
]
hierarchy [
 root {loc=<0 0 0>, rot=<0 0 0 1>}
 pack {parent=root, loc=<0 -0.7316 0>, rot=<0 0 0 1>}
 dyn_spine_1 {parent=pack, loc=<5.1226 -7.089 4.698>, rot=<-0.3953 -0.5863 0.5863 -0.3953>}
 dyn_spine_2 {parent=dyn_spine_1, loc=<11.2974 0 0>, rot=<-0.0037 -0.053 0.0748 -0.9958>}
 dyn_spine_3 {parent=dyn_spine_2, loc=<10.9658 0 0>, rot=<0 0 0.0193 -0.9998>}
 dyn_tile_1 {parent=dyn_spine_3, loc=<8.4709 -0.1172 0>, rot=<-0.0081 -0.0323 0.2423 -0.9696>}
 dyn_tile_2 {parent=dyn_tile_1, loc=<7.2026 -0.9652 -0.1205>, rot=<-0.0001 -0.0064 0.0191 -0.9998>}
 dyn_tile_3 {parent=dyn_tile_2, loc=<6.3297 0 0>, rot=<0 0 0 1>}
 dyn_R_small_fin {parent=dyn_spine_2, loc=<8.3537 0.8547 -4.4835>, rot=<-0.0215 -0.6142 0.0219 -0.7886>}
 dyn_R_fin {parent=dyn_spine_1, loc=<6.6475 2.3566 -6.1017>, rot=<-0.0745 -0.7032 0.0745 -0.7032>}
 dyn_L_fin {parent=dyn_spine_1, loc=<6.7606 2.7735 6.2625>, rot=<0.0745 0.7032 0.0745 -0.7032>}
 dyn_jaw_1 {parent=dyn_spine_1, loc=<0.3274 3.2565 0>, rot=<0 0 0.9976 0.0695>}
 dyn_jaw_2 {parent=dyn_jaw_1, loc=<1.5845 0 0>, rot=<0 0 0 1>}
]
