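# Animation Blueprint for the Cactus Rocker backpack, procedurally posing and jiggling its spine, tile, fin and jaw bones.
# From the ref pose, it layers skydive root offset, interpolated pitch, fixed bone rotations and AnimDynamics, finishing in a Control Rig.
# The EventGraph derives skydive/melee pitch plus movement, jump and direction weights from player state.
# Several alternative AnimDynamics setups, including the chain using skydiving_alpha, are disconnected from the output.
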AnimBlueprint /BRCosmetics/Accessories/FORT_Backpacks/Backpack_M_MED_Cactus_Rocker/Meshes/M_MED_Cactus_Rocker_Pack_AnimBP
target_skeleton=&/M_MED_Cactus_Rocker_Pack_Skeleton
graphs [
 AnimGraph {nodes [
   "Anim Dynamics - Chain: dyn_spine_2 -> dyn_tile_3"(Alpha=Get skydiving_alpha() -> (skydiving_alpha:double), ExternalForce="0,0,0") {
    angularDampingOverride=1
    bAngularSpring=true
    bChain=true
    bGravityOverrideInSimSpace=true
    bOverrideAngularDamping=true
    bUseGravityOverride=true
    simulationSpace=Actor
    alphaBoolBlend {blendOption=Linear}
    boundBone {boneName=dyn_spine_2}
    chainEnd {boneName=dyn_tile_3}
    gravityOverride {z=50}
    physicsBodyDefinitions [
     {
      boundBone {boneName=dyn_spine_2}
      boxExtents=<15 40 40>
      constraintSetup {angularLimitsMax=<0 1 4>, angularLimitsMin=<0 -3 -4>, angularTarget=<1 0 0>}
      localJointOffset=<0 0 0>
     }
     {
      boundBone {boneName=dyn_spine_3}
      boxExtents=<13 40 40>
      constraintSetup {angularLimitsMax=<0 1 4>, angularLimitsMin=<0 -3 -4>, angularTarget=<0 0 0>}
      localJointOffset=<0 0 0>
     }
     {
      boundBone {boneName=dyn_tile_1}
      boxExtents=<10 25 25>
      constraintSetup {angularLimitsMax=<0 1 5>, angularLimitsMin=<0 -5 -15>, angularTarget=<0 0 0>}
      localJointOffset=<5 5 0>
     }
     {
      boundBone {boneName=dyn_tile_2}
      boxExtents=<10 20 20>
      constraintSetup {angularLimitsMax=<0 1 5>, angularLimitsMin=<0 -5 -5>, angularTarget=<0 0 0>}
      localJointOffset=<0 0 0>
     }
     {
      sphereCollisionRadius=5
      boundBone {boneName=dyn_tile_3}
      boxExtents=<10 20 20>
      constraintSetup {angularLimitsMax=<0 1 5>, angularLimitsMin=<0 -5 -5>, angularTarget=<0 0 0>}
      localJointOffset=<0 0 0>
     }
    ]
   }
   Anim Dynamics - Bone: dyn_tile_3!(ComponentPose, Alpha=1, GravityOverride:Vector="0,0,-980", ExternalForce="0,0,0") -> (Pose) {
    angularDampingOverride=0.99
    angularSpringConstant=100
    bAngularSpring=true
    bOverrideAngularDamping=true
    simulationSpace=BoneRelative
    alphaBoolBlend {blendOption=Linear}
    alphaScaleBiasClamp {bInterpResult=true, interpSpeedDecreasing=5, interpSpeedIncreasing=7}
    boundBone {boneName=dyn_tile_3}
    componentAppliedLinearAccClamp=<3500 3500 3500>
    componentLinearAccScale=<0.8 0.8 0.8>
    componentLinearVelScale=<0.8 0.8 0.8>
    gravityOverride {z=-980}
    physicsBodyDefinitions [{
      collisionType=CustomSphere
      sphereCollisionRadius=4.5
      boundBone {boneName=dyn_tile_3}
      boxExtents=<25 15 15>
      constraintSetup {angularLimitsMax=<1 1 30>, angularLimitsMin=<-1 -20 -35>, angularTarget=<0 0 0>}
      localJointOffset=<2 0 0>
     }]
    relativeSpaceBone {boneName=pack}
   }
   n1 = Anim Dynamics - Bone: dyn_spine_2(Alpha=1, ExternalForce="0,0,0") {
    angularDampingOverride=1
    angularSpringConstant=500
    bAngularSpring=true
    bGravityOverrideInSimSpace=true
    bOverrideAngularDamping=true
    bUseGravityOverride=true
    linearSpringConstant=200
    simulationSpace=BoneRelative
    alphaBoolBlend {blendOption=Linear}
    boundBone {boneName=dyn_spine_2}
    componentLinearAccScale=<1.2 1.2 1.2>
    componentLinearVelScale=<1.2 1.2 1.2>
    gravityOverride {z=-360}
    physicsBodyDefinitions [{
      boundBone {boneName=dyn_spine_2}
      boxExtents=<15 40 40>
      constraintSetup {angularLimitsMax=<0 1 40>, angularLimitsMin=<0 -3 -40>, angularTarget=<1 0 0>}
      localJointOffset=<5 5 0>
     }]
    relativeSpaceBone {boneName=root}
   }
   n2 = Anim Dynamics - Bone: dyn_spine_3(ComponentPose=@n1, Alpha=1, ExternalForce="0,0,0") {
    angularDampingOverride=1
    angularSpringConstant=500
    bAngularSpring=true
    bGravityOverrideInSimSpace=true
    bOverrideAngularDamping=true
    bUseGravityOverride=true
    linearSpringConstant=200
    simulationSpace=BoneRelative
    alphaBoolBlend {blendOption=Linear}
    boundBone {boneName=dyn_spine_3}
    componentLinearAccScale=<1.2 1.2 1.2>
    componentLinearVelScale=<1.2 1.2 1.2>
    gravityOverride {z=-360}
    physicsBodyDefinitions [{
      boundBone {boneName=dyn_spine_3}
      boxExtents=<13 40 40>
      constraintSetup {angularLimitsMax=<0 1 30>, angularLimitsMin=<0 -3 -30>, angularTarget=<0 0 0>}
      localJointOffset=<5 5 0>
     }]
    relativeSpaceBone {boneName=root}
   }
   n3 = Anim Dynamics - Bone: dyn_tile_1(ComponentPose=@n2, Alpha=1, ExternalForce="0,0,0") {
    angularDampingOverride=1
    angularSpringConstant=500
    bAngularSpring=true
    bGravityOverrideInSimSpace=true
    bOverrideAngularDamping=true
    bUseGravityOverride=true
    linearSpringConstant=200
    simulationSpace=BoneRelative
    alphaBoolBlend {blendOption=Linear}
    boundBone {boneName=dyn_tile_1}
    componentLinearAccScale=<1.2 1.2 1.2>
    componentLinearVelScale=<1.2 1.2 1.2>
    gravityOverride {z=-360}
    physicsBodyDefinitions [{
      boundBone {boneName=dyn_tile_1}
      boxExtents=<10 25 25>
      constraintSetup {angularLimitsMax=<0 1 25>, angularLimitsMin=<0 -5 -25>, angularTarget=<0 0 0>}
      localJointOffset=<5 5 0>
     }]
    relativeSpaceBone {boneName=root}
   }
   n4 = Anim Dynamics - Bone: dyn_tile_2(ComponentPose=@n3, Alpha=1, ExternalForce="0,0,0") {
    angularDampingOverride=1
    angularSpringConstant=500
    bAngularSpring=true
    bGravityOverrideInSimSpace=true
    bOverrideAngularDamping=true
    bUseGravityOverride=true
    linearSpringConstant=200
    simulationSpace=BoneRelative
    alphaBoolBlend {blendOption=Linear}
    boundBone {boneName=dyn_tile_2}
    componentLinearAccScale=<1.2 1.2 1.2>
    componentLinearVelScale=<1.2 1.2 1.2>
    gravityOverride {z=-360}
    physicsBodyDefinitions [{
      boundBone {boneName=dyn_tile_2}
      boxExtents=<10 20 20>
      constraintSetup {angularLimitsMax=<0 1 20>, angularLimitsMin=<0 -5 -20>, angularTarget=<0 0 0>}
      localJointOffset=<0 0 0>
     }]
    relativeSpaceBone {boneName=root}
   }
   Anim Dynamics - Bone: dyn_tile_3!(ComponentPose=@n4, Alpha=1, ExternalForce="0,0,0") -> (Pose) {
    angularDampingOverride=1
    angularSpringConstant=500
    bAngularSpring=true
    bGravityOverrideInSimSpace=true
    bOverrideAngularDamping=true
    bUseGravityOverride=true
    linearSpringConstant=200
    simulationSpace=BoneRelative
    alphaBoolBlend {blendOption=Linear}
    boundBone {boneName=dyn_tile_3}
    componentLinearAccScale=<1.2 1.2 1.2>
    componentLinearVelScale=<1.2 1.2 1.2>
    gravityOverride {z=-360}
    physicsBodyDefinitions [{
      sphereCollisionRadius=5
      boundBone {boneName=dyn_tile_3}
      boxExtents=<10 20 20>
      constraintSetup {angularLimitsMax=<0 1 15>, angularLimitsMin=<0 -5 -15>, angularTarget=<0 0 0>}
      localJointOffset=<0 0 0>
     }]
    relativeSpaceBone {boneName=root}
   }
   Anim Dynamics - Bone: dyn_tails_bott_r_01(Alpha=1, ExternalForce="0,0,0") {
    angularDampingOverride=0.9999
    angularSpringConstant=500
    bAngularSpring=true
    bGravityOverrideInSimSpace=true
    bOverrideAngularDamping=true
    simulationSpace=BoneRelative
    alphaBoolBlend {blendOption=Linear}
    boundBone {boneName=dyn_tails_bott_r_01}
    componentAppliedLinearAccClamp=<3500 3500 3500>
    componentLinearAccScale=<1.2 1.2 1.2>
    componentLinearVelScale=<1.2 1.2 1.2>
    physicsBodyDefinitions [{
      sphereCollisionRadius=0
      boundBone {boneName=dyn_tails_bott_r_01}
      boxExtents=<15 30 30>
      constraintSetup {angularLimitsMax=<1 5 15>, angularLimitsMin=<-1 -35 -15>, angularTarget=<0 0 0>}
      localJointOffset=<-3 0 0>
     }]
    relativeSpaceBone {boneName=root}
    sphericalLimits [{limitRadius=6, drivingBone {boneName=dyn_back_flap_02}, sphereLocalOffset=<0 0 0>}]
   }
   n5 = Anim Dynamics - Bone: dyn_spine_2(Alpha=1, ExternalForce="0,0,0") {
    angularDampingOverride=1
    angularSpringConstant=240
    bAngularSpring=true
    bLinearSpring=true
    bOverrideAngularDamping=true
    bUseGravityOverride=true
    linearSpringConstant=20
    simulationSpace=BoneRelative
    alphaBoolBlend {blendOption=Linear}
    boundBone {boneName=dyn_spine_2}
    componentAppliedLinearAccClamp=<3000 3000 3000>
    componentLinearAccScale=<0.1 0.1 0.3>
    componentLinearVelScale=<0.1 0.1 0.1>
    gravityOverride {z=-360}
    physicsBodyDefinitions [{
      boundBone {boneName=dyn_spine_2}
      boxExtents=<24 32 32>
      constraintSetup {angularLimitsMax=<1 2 10>, angularLimitsMin=<-1 -10 -40>, angularTarget=<1 0 0>}
      localJointOffset=<12 0 0>
     }]
    relativeSpaceBone {boneName=pack}
   }
   n6 = Anim Dynamics - Bone: dyn_spine_3(ComponentPose=@n5, Alpha=1, ExternalForce="0,0,0") {
    angularDampingOverride=1
    angularSpringConstant=240
    bAngularSpring=true
    bLinearSpring=true
    bOverrideAngularDamping=true
    bUseGravityOverride=true
    linearSpringConstant=20
    simulationSpace=BoneRelative
    alphaBoolBlend {blendOption=Linear}
    boundBone {boneName=dyn_spine_3}
    componentAppliedLinearAccClamp=<3000 3000 3000>
    componentLinearAccScale=<0.1 0.1 0.3>
    componentLinearVelScale=<0.1 0.1 0.1>
    gravityOverride {z=-360}
    physicsBodyDefinitions [{
      boundBone {boneName=dyn_spine_3}
      boxExtents=<24 32 32>
      constraintSetup {angularLimitsMax=<1 2 10>, angularLimitsMin=<-1 -10 -40>, angularTarget=<1 0 0>}
      localJointOffset=<12 0 0>
     }]
    relativeSpaceBone {boneName=pack}
   }
   n7 = Anim Dynamics - Bone: dyn_tile_1(ComponentPose=@n6, Alpha=1, ExternalForce="0,0,0") {
    angularDampingOverride=1
    angularSpringConstant=240
    bAngularSpring=true
    bLinearSpring=true
    bOverrideAngularDamping=true
    bUseGravityOverride=true
    linearSpringConstant=20
    simulationSpace=BoneRelative
    alphaBoolBlend {blendOption=Linear}
    boundBone {boneName=dyn_tile_1}
    componentAppliedLinearAccClamp=<3000 3000 3000>
    componentLinearAccScale=<0.1 0.1 0.3>
    componentLinearVelScale=<0.1 0.1 0.1>
    gravityOverride {z=-360}
    physicsBodyDefinitions [{
      boundBone {boneName=dyn_tile_1}
      boxExtents=<24 32 32>
      constraintSetup {angularLimitsMax=<1 2 10>, angularLimitsMin=<-1 -10 -40>, angularTarget=<1 0 0>}
      localJointOffset=<12 0 0>
     }]
    relativeSpaceBone {boneName=pack}
   }
   n8 = Anim Dynamics - Bone: dyn_tile_2(ComponentPose=@n7, Alpha=1, ExternalForce="0,0,0") {
    angularDampingOverride=1
    angularSpringConstant=240
    bAngularSpring=true
    bLinearSpring=true
    bOverrideAngularDamping=true
    bUseGravityOverride=true
    linearSpringConstant=20
    simulationSpace=BoneRelative
    alphaBoolBlend {blendOption=Linear}
    boundBone {boneName=dyn_tile_2}
    componentAppliedLinearAccClamp=<3000 3000 3000>
    componentLinearAccScale=<0.1 0.1 0.3>
    componentLinearVelScale=<0.1 0.1 0.1>
    gravityOverride {z=-360}
    physicsBodyDefinitions [{
      boundBone {boneName=dyn_tile_2}
      boxExtents=<24 32 32>
      constraintSetup {angularLimitsMax=<1 2 10>, angularLimitsMin=<-1 -10 -40>, angularTarget=<1 0 0>}
      localJointOffset=<12 0 0>
     }]
    relativeSpaceBone {boneName=pack}
   }
   Anim Dynamics - Bone: dyn_tile_3(ComponentPose=@n8, Alpha=1, ExternalForce="0,0,0") {
    angularDampingOverride=1
    angularSpringConstant=240
    bAngularSpring=true
    bLinearSpring=true
    bOverrideAngularDamping=true
    bUseGravityOverride=true
    linearSpringConstant=20
    simulationSpace=BoneRelative
    alphaBoolBlend {blendOption=Linear}
    boundBone {boneName=dyn_tile_3}
    componentAppliedLinearAccClamp=<3000 3000 3000>
    componentLinearAccScale=<0.1 0.1 0.3>
    componentLinearVelScale=<0.1 0.1 0.1>
    gravityOverride {z=-360}
    physicsBodyDefinitions [{
      boundBone {boneName=dyn_tile_3}
      boxExtents=<24 32 32>
      constraintSetup {angularLimitsMax=<1 2 10>, angularLimitsMin=<-1 -10 -40>, angularTarget=<1 0 0>}
      localJointOffset=<12 0 0>
     }]
    relativeSpaceBone {boneName=pack}
   }
   n9 = "Transform (Modify) Bone - Bone: root"(ComponentPose=Local To Component(LocalPose=Local Space Ref Pose()), Translation="0,1,0", Rotation="0,0,0") {translationMode=BMM_Additive, alphaBoolBlend {blendOption=Linear}, boneToModify {boneName=root}, translation {y=1}}
   n10 = "Transform (Modify) Bone - Bone: root"!(ComponentPose=@n9, bAlphaBoolEnabled:bool=Get bIsSkyDiving(), Translation="0,-1,0", Rotation="0,0,0", Scale) -> (Pose) {alphaInputType=Bool, translationMode=BMM_Additive, alphaBoolBlend {blendOption=Linear}, boneToModify {boneName=root}, translation {y=-1}}
   n11 = Get skydiving_rotation() -> (skydiving_rotation:double)
   n12 = "Transform (Modify) Bone - Bone: dyn_spine_2"!(ComponentPose=@n10, Alpha=1, Translation="0,0,0", Rotation_Roll:float=0, Rotation_Pitch:float=@n11, Rotation_Yaw:float=0, Scale="1,1,1") -> (Pose) {
    rotationMode=BMM_Replace
    rotationSpace=BCS_BoneSpace
    alphaBoolBlend {blendOption=Linear}
    boneToModify {boneName=dyn_spine_2}
    rotation {pitch=30}
   }
   n13 = "Transform (Modify) Bone - Bone: dyn_spine_3"!(ComponentPose=@n12, Alpha=1, Translation="0,0,0", Rotation_Roll:float=0, Rotation_Pitch:float=@n11, Rotation_Yaw:float=0, Scale="1,1,1") -> (Pose) {
    rotationMode=BMM_Replace
    rotationSpace=BCS_BoneSpace
    alphaBoolBlend {blendOption=Linear}
    boneToModify {boneName=dyn_spine_3}
    rotation {pitch=30}
   }
   n14 = "Transform (Modify) Bone - Bone: dyn_tile_1"!(ComponentPose=@n13, Alpha=1, Translation="0,0,0", Rotation_Roll:float=0, Rotation_Pitch:float=@n11, Rotation_Yaw:float=0, Scale="1,1,1") -> (Pose) {
    rotationMode=BMM_Replace
    rotationSpace=BCS_BoneSpace
    alphaBoolBlend {blendOption=Linear}
    boneToModify {boneName=dyn_tile_1}
    rotation {pitch=30}
   }
   n15 = "Transform (Modify) Bone - Bone: dyn_spine_2"!(ComponentPose=@n14, Alpha=1, Rotation:Rotator="5.632,-15.419,-0.389") -> (Pose) {
    rotationMode=BMM_Replace
    rotationSpace=BCS_BoneSpace
    alphaBoolBlend {blendOption=Linear}
    alphaScaleBiasClamp {bInterpResult=true, clampMax=0.5, interpSpeedDecreasing=7}
    boneToModify {boneName=dyn_spine_2}
    rotation=<rot 5.632 -15.419 -0.389>
   }
   n16 = "Transform (Modify) Bone - Bone: dyn_spine_3"!(ComponentPose=@n15, Alpha=1, Rotation:Rotator="1.643,-12.662,-2.548") -> (Pose) {
    rotationMode=BMM_Replace
    rotationSpace=BCS_BoneSpace
    alphaBoolBlend {blendOption=Linear}
    alphaScaleBiasClamp {bInterpResult=true, clampMax=0.5, interpSpeedDecreasing=7}
    boneToModify {boneName=dyn_spine_3}
    rotation=<rot 1.643 -12.662 -2.548>
   }
   n17 = "Transform (Modify) Bone - Bone: dyn_tile_1"!(ComponentPose=@n16, Alpha=1, Rotation:Rotator="0,-35,0") -> (Pose) {
    rotationMode=BMM_Replace
    rotationSpace=BCS_BoneSpace
    alphaBoolBlend {blendOption=Linear}
    alphaScaleBiasClamp {bInterpResult=true, clampMax=0.5, interpSpeedDecreasing=7}
    boneToModify {boneName=dyn_tile_1}
    rotation {yaw=-35}
   }
   n18 = Anim Dynamics - Bone: dyn_spine_2!(ComponentPose=@n17, Alpha=1, GravityOverride:Vector="0,0,-980", ExternalForce="0,0,0") -> (Pose) {
    angularDampingOverride=0.99
    angularSpringConstant=100
    bAngularSpring=true
    bOverrideAngularDamping=true
    simulationSpace=BoneRelative
    alphaBoolBlend {blendOption=Linear}
    alphaScaleBiasClamp {bInterpResult=true, interpSpeedDecreasing=5, interpSpeedIncreasing=1}
    boundBone {boneName=dyn_spine_2}
    componentAppliedLinearAccClamp=<3500 3500 3500>
    componentLinearAccScale=<0.5 0.5 0.5>
    componentLinearVelScale=<0.5 0.5 0.5>
    gravityOverride {z=-980}
    physicsBodyDefinitions [{
      collisionType=CustomSphere
      sphereCollisionRadius=4.5
      boundBone {boneName=dyn_spine_2}
      boxExtents=<25 15 15>
      constraintSetup {angularLimitsMax=<1 1 50>, angularLimitsMin=<-1 -10 -20>, angularTarget=<10 0 0>}
      localJointOffset=<5 0 0>
     }]
    relativeSpaceBone {boneName=pack}
   }
   n19 = Anim Dynamics - Bone: dyn_spine_3!(ComponentPose=@n18, Alpha=1, GravityOverride:Vector="0,0,-980", ExternalForce="0,0,0") -> (Pose) {
    angularDampingOverride=0.99
    angularSpringConstant=100
    bAngularSpring=true
    bOverrideAngularDamping=true
    simulationSpace=BoneRelative
    alphaBoolBlend {blendOption=Linear}
    alphaScaleBiasClamp {bInterpResult=true, interpSpeedDecreasing=5, interpSpeedIncreasing=1}
    boundBone {boneName=dyn_spine_3}
    componentAppliedLinearAccClamp=<3500 3500 3500>
    componentLinearAccScale=<0.5 0.5 0.5>
    componentLinearVelScale=<0.5 0.5 0.5>
    gravityOverride {z=-980}
    physicsBodyDefinitions [{
      collisionType=CustomSphere
      sphereCollisionRadius=4.5
      boundBone {boneName=dyn_spine_3}
      boxExtents=<25 15 15>
      constraintSetup {angularLimitsMax=<1 1 45>, angularLimitsMin=<-1 -10 -20>, angularTarget=<0 0 0>}
      localJointOffset=<2 0 0>
     }]
    relativeSpaceBone {boneName=pack}
   }
   n20 = Anim Dynamics - Bone: dyn_tile_1!(ComponentPose=@n19, Alpha=1, GravityOverride:Vector="0,0,-980", ExternalForce="0,0,0") -> (Pose) {
    angularDampingOverride=0.99
    angularSpringConstant=100
    bAngularSpring=true
    bOverrideAngularDamping=true
    simulationSpace=BoneRelative
    alphaBoolBlend {blendOption=Linear}
    alphaScaleBiasClamp {bInterpResult=true, interpSpeedDecreasing=5, interpSpeedIncreasing=1}
    boundBone {boneName=dyn_tile_1}
    componentAppliedLinearAccClamp=<3500 3500 3500>
    componentLinearAccScale=<0.8 0.8 0.8>
    componentLinearVelScale=<0.8 0.8 0.8>
    gravityOverride {z=-980}
    physicsBodyDefinitions [{
      collisionType=CustomSphere
      sphereCollisionRadius=4.5
      boundBone {boneName=dyn_tile_1}
      boxExtents=<25 15 15>
      constraintSetup {angularLimitsMax=<1 1 20>, angularLimitsMin=<-1 -20 -20>, angularTarget=<0 0 0>}
      localJointOffset=<2 0 0>
     }]
    relativeSpaceBone {boneName=pack}
   }
   n21 = Anim Dynamics - Bone: dyn_tile_2!(ComponentPose=@n20, Alpha=1, GravityOverride:Vector="0,0,-980", ExternalForce="0,0,0") -> (Pose) {
    angularDampingOverride=0.99
    angularSpringConstant=100
    bAngularSpring=true
    bOverrideAngularDamping=true
    simulationSpace=BoneRelative
    alphaBoolBlend {blendOption=Linear}
    alphaScaleBiasClamp {bInterpResult=true, interpSpeedDecreasing=5, interpSpeedIncreasing=1}
    boundBone {boneName=dyn_tile_2}
    componentAppliedLinearAccClamp=<3500 3500 3500>
    componentLinearAccScale=<0.8 0.8 0.8>
    componentLinearVelScale=<0.8 0.8 0.8>
    gravityOverride {z=-980}
    physicsBodyDefinitions [{
      collisionType=CustomSphere
      sphereCollisionRadius=4.5
      boundBone {boneName=dyn_tile_2}
      boxExtents=<25 15 15>
      constraintSetup {angularLimitsMax=<1 1 20>, angularLimitsMin=<-1 -20 -20>, angularTarget=<0 0 0>}
      localJointOffset=<2 0 0>
     }]
    relativeSpaceBone {boneName=pack}
   }
   n22 = Anim Dynamics - Bone: dyn_L_fin(ComponentPose=@n21, Alpha=1, ExternalForce="0,0,0") {
    angularDampingOverride=0.99
    angularSpringConstant=300
    bAngularSpring=true
    bLinearSpring=true
    bOverrideAngularDamping=true
    linearSpringConstant=100
    simulationSpace=BoneRelative
    alphaBoolBlend {blendOption=Linear}
    boundBone {boneName=dyn_L_fin}
    physicsBodyDefinitions [{
      boundBone {boneName=dyn_L_fin}
      boxExtents=<30 10 30>
      constraintSetup {angularLimitsMax=<0 0 30>, angularLimitsMin=<0 0 -30>, angularTarget=<0 0 0>}
      localJointOffset=<0 16 -20>
     }]
    relativeSpaceBone {boneName=root}
   }
   n23 = "Transform (Modify) Bone - Bone: dyn_spine_2"(ComponentPose=@n22, Alpha=1, Rotation="-0.064,-42.813,-0.135") {
    rotationMode=BMM_Replace
    rotationSpace=BCS_BoneSpace
    alphaBoolBlend {blendOption=Linear}
    alphaScaleBiasClamp {bInterpResult=true, clampMax=0.5, interpSpeedDecreasing=7, interpSpeedIncreasing=1}
    boneToModify {boneName=dyn_spine_2}
    rotation=<rot -0.064 -42.813 -0.135>
   }
   n24 = "Transform (Modify) Bone - Bone: dyn_spine_3"(ComponentPose=@n23, Alpha=1, Rotation="0.690,-17.250,-1.102") {
    rotationMode=BMM_Replace
    rotationSpace=BCS_BoneSpace
    alphaBoolBlend {blendOption=Linear}
    alphaScaleBiasClamp {bInterpResult=true, clampMax=0.5, interpSpeedDecreasing=7, interpSpeedIncreasing=1}
    boneToModify {boneName=dyn_spine_3}
    rotation=<rot 0.69 -17.25 -1.102>
   }
   n25 = "Transform (Modify) Bone - Bone: dyn_tile_1"(ComponentPose=@n24, Alpha=1, Rotation="-1.347,-3.149,0.232") {
    rotationMode=BMM_Replace
    rotationSpace=BCS_BoneSpace
    alphaBoolBlend {blendOption=Linear}
    alphaScaleBiasClamp {bInterpResult=true, clampMax=0.5, interpSpeedDecreasing=7, interpSpeedIncreasing=1}
    boneToModify {boneName=dyn_tile_1}
    rotation=<rot -1.347 -3.149 0.232>
   }
   n26 = "Transform (Modify) Bone - Bone: dyn_spine_2"(ComponentPose=@n25, Alpha=1, Rotation="-0.064,-42.813,-0.135") {
    rotationMode=BMM_Replace
    rotationSpace=BCS_BoneSpace
    alphaBoolBlend {blendOption=Linear}
    alphaScaleBiasClamp {bInterpResult=true, clampMax=0.5, interpSpeedDecreasing=7, interpSpeedIncreasing=1}
    boneToModify {boneName=dyn_spine_2}
    rotation=<rot -0.064 -42.813 -0.135>
   }
   n27 = "Transform (Modify) Bone - Bone: dyn_spine_3"(ComponentPose=@n26, Alpha=1, Rotation="0.690,-17.250,-1.102") {
    rotationMode=BMM_Replace
    rotationSpace=BCS_BoneSpace
    alphaBoolBlend {blendOption=Linear}
    alphaScaleBiasClamp {bInterpResult=true, clampMax=0.5, interpSpeedDecreasing=7, interpSpeedIncreasing=1}
    boneToModify {boneName=dyn_spine_3}
    rotation=<rot 0.69 -17.25 -1.102>
   }
   n28 = "Transform (Modify) Bone - Bone: dyn_tile_1"(ComponentPose=@n27, Alpha=1, Rotation="-1.347,-3.149,0.232") {
    rotationMode=BMM_Replace
    rotationSpace=BCS_BoneSpace
    alphaBoolBlend {blendOption=Linear}
    alphaScaleBiasClamp {bInterpResult=true, clampMax=0.5, interpSpeedDecreasing=7, interpSpeedIncreasing=1}
    boneToModify {boneName=dyn_tile_1}
    rotation=<rot -1.347 -3.149 0.232>
   }
   n29 = "Transform (Modify) Bone - Bone: dyn_spine_2"(ComponentPose=@n28, Alpha=1, Rotation="-0.064,-42.813,-0.135") {
    rotationMode=BMM_Replace
    rotationSpace=BCS_BoneSpace
    alphaBoolBlend {blendOption=Linear}
    alphaScaleBiasClamp {bInterpResult=true, clampMax=0.5, interpSpeedDecreasing=7, interpSpeedIncreasing=1}
    boneToModify {boneName=dyn_spine_2}
    rotation=<rot -0.064 -42.813 -0.135>
   }
   n30 = "Transform (Modify) Bone - Bone: dyn_spine_3"(ComponentPose=@n29, Alpha=1, Rotation="0.690,-17.250,-1.102") {
    rotationMode=BMM_Replace
    rotationSpace=BCS_BoneSpace
    alphaBoolBlend {blendOption=Linear}
    alphaScaleBiasClamp {bInterpResult=true, clampMax=0.5, interpSpeedDecreasing=7, interpSpeedIncreasing=1}
    boneToModify {boneName=dyn_spine_3}
    rotation=<rot 0.69 -17.25 -1.102>
   }
   n31 = "Transform (Modify) Bone - Bone: dyn_tile_1"(ComponentPose=@n30, Alpha=1, Rotation="-1.347,-3.149,0.232") {
    rotationMode=BMM_Replace
    rotationSpace=BCS_BoneSpace
    alphaBoolBlend {blendOption=Linear}
    alphaScaleBiasClamp {bInterpResult=true, clampMax=0.5, interpSpeedDecreasing=7, interpSpeedIncreasing=1}
    boneToModify {boneName=dyn_tile_1}
    rotation=<rot -1.347 -3.149 0.232>
   }
   n32 = "Transform (Modify) Bone - Bone: dyn_R_fin"(ComponentPose=@n31, Alpha=1, Translation="0,0,0", Rotation="0,70,0", Scale="1,1,1") {
    rotationMode=BMM_Replace
    rotationSpace=BCS_BoneSpace
    alphaBoolBlend {blendOption=Linear}
    boneToModify {boneName=dyn_R_fin}
    rotation {yaw=70}
   }
   n33 = "Transform (Modify) Bone - Bone: dyn_R_small_fin"(ComponentPose=@n32, Alpha=1, Translation="0,0,0", Rotation="0,30,0", Scale="1,1,1") {
    rotationMode=BMM_Replace
    rotationSpace=BCS_BoneSpace
    alphaBoolBlend {blendOption=Linear}
    boneToModify {boneName=dyn_R_small_fin}
    rotation {yaw=30}
   }
   n34 = "Transform (Modify) Bone - Bone: dyn_jaw_2"(ComponentPose=@n33, Alpha=1, Translation="0,0,0", Rotation="-20,0,0", Scale="1,1,1") {rotationMode=BMM_Additive, alphaBoolBlend {blendOption=Linear}, boneToModify {boneName=dyn_jaw_2}, rotation {pitch=-20}}
   n35 = Control Rig(Source=Component To Local(ComponentPose=@n34)) {
    alphaBoolBlend {blendOption=Linear}
    controlRigAssetReference {blueprintRigClass=&/M_MED_Cactus_Rocker_Pack_CtrlRig.M_MED_Cactus_Rocker_Pack_CtrlRig_C}
   }
   Output Pose(Result=@n35) {layerGroup=None}
  ]}
 EventGraph {nodes [
   n1 = Event BlueprintUpdateAnimation()
   n2 = Get bIsSkyDiving()
   n3 = NOT Boolean(A=@n2)
   n4 = Get bIsMoving2D()
   > Set skydiving_alpha(skydiving_alpha:double=SelectFloat(A=1, B=0.2, bPickA=AND Boolean(A=@n4, B=@n3))) -> (Output_Get:double)
   n5 = NOT Boolean(A=Get bIsFalling())
   n6 = NOT Boolean(A=Get bIsJumping())
   n7 = Get bIsMoving2D()
   > Set Moving(Moving:double=SelectFloat(A=0.2, B=1, bPickA=AND Boolean!(A=@n7, B=@n6, C:bool=@n5) -> (ReturnValue))) -> (Output_Get:double)
   n8 = NOT Boolean(A=Get bIsSkyDiving())
   n9 = OR Boolean!(A=Get bIsJumping(), B=Get bIsFalling(), C:bool) -> (ReturnValue)
   > Set Jumps(Jumps:double=SelectFloat(A=1, B=0, bPickA=AND Boolean!(A=@n8, B=@n9) -> (ReturnValue))) -> (Output_Get:double)
   n10 = SelectFloat(A=25, B=0, bPickA=@n2)
   n11 = Get bIsPlayingMeleeAnim()
   > Set skydiving_rotation(skydiving_rotation:double=FInterpTo(Current=Get skydiving_rotation(), Target=SelectFloat(A=20, B=@n10, bPickA=@n11), DeltaTime=@n1.DeltaTimeX, InterpSpeed=0.15)) -> (Output_Get:double)
   > n12 = PlayerMovement(AnimInstance=Self-Reference(), XDotProductCoefficient=100, YDotProductCoefficient=100, ZDotProductCoefficient=100)
   n13 = BreakVector(InVec=MakeVector(X=@n12, Y=@n12.YMovement))
   > Set Left(Left:double="To Float (Boolean)"(InBool="float < float"(A=@n13, B=-50))) -> (Output_Get:double)
   > Set Right(Right:double="To Float (Boolean)"(InBool="float > float"(A=@n13, B=50))) -> (Output_Get:double)
   > Set Forward(Forward:double="To Float (Boolean)"(InBool="float > float"(A=@n13.Y, B=40))) -> (Output_Get:double)
   > Set Backward(Backward:double="To Float (Boolean)"(InBool="float < float"(A=@n13.Y, B=-40))) -> (Output_Get:double)
  ]}
]
variables [
 Backward {type=number}
 Forward {type=number}
 Jumps {type=number}
 Left {type=number}
 melee_aim {type=boolean}
 Moving {type=number}
 Right {type=number}
 skydiving_alpha {type=number}
 skydiving_rotation {type=number}
]
